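# Control Rig for the BrightShimmer Casual female body, applying corrective bone adjustments to the belt and upper arms.
# Four chained spherical pose readers on thigh_r weight relative offsets on belt_corrective_01/02 and dyn_belt.
# When "is lod2" is true, it parent-constrains both upperarm_twist_01 bones, then offsets them by remapped upperarm_l rotation relative to spine_01, even the right side.
# Several alternative belt transform nodes are left unexecuted.
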ControlRigBlueprint /BRCosmetics/Characters/Player/Female/Medium/Bodies/F_MED_BrightShimmer_Casual/Meshes/F_MED_BrightShimmer_Casual_CtrlRig
graphs [
 RigVMFunctionLibrary {links [], nodes []}
 RigVMModel {nodes [
   n1 = /Script/RigVM.RigVMFunction_MathTransformFromSRT(Location=(X=0,Y=0,Z=10), Rotation=(X=0,Y=0,Z=0), Scale=(X=1,Y=1,Z=1)) {is=struct RigVMUnitNode}
   n2 = &RigUnit_SphericalPoseReader(
    ActiveRegionScaleFactors=(PositiveWidth=1,NegativeWidth=1.5,PositiveHeight=1,NegativeHeight=1)
    ActiveRegionSize=0.03
    Debug=(bDrawDebug=true,DebugScale=25,DebugSegments=20,DebugThickness=0.25)
    DriverAxis=(X=1,Y=0,Z=0)
    DriverItem=(Type=Bone,Name="thigh_r")
    FalloffRegionScaleFactors=(PositiveWidth=2,NegativeWidth=1.5,PositiveHeight=1,NegativeHeight=1.5)
    FalloffSize=0.1
    OptionalParentItem=(Type=Bone)
    RotationOffset=(X=-65,Y=0,Z=40)
    ExecutePin=@n9.ExecutePin
   )
   &RigUnit_ModifyTransforms(
    ExecutePin=()
    ItemToModify=((Item=(Type=Bone,Name="belt_corrective_01"),Transform=(Rotation=(X=0,Y=0,Z=0,W=1),Translation=(X=0,Y=8,Z=0),Scale3D=(X=1,Y=1,Z=1))))
    ItemToModify.0.Transform=@n1.Transform
    Weight=@n2.OutputParam
   )
   n3 = &RigVMFunction_MathTransformFromSRT(Location=(X=0,Y=3,Z=0), Rotation=(X=0,Y=0,Z=0), Scale=(X=1,Y=1,Z=1))
   > &RigUnit_ModifyTransforms(
    ItemToModify=((Item=(Type=Bone,Name="belt_corrective_02"),Transform=(Rotation=(X=0,Y=0,Z=0,W=1),Translation=(X=0,Y=8,Z=0),Scale3D=(X=1,Y=1,Z=1))))
    ItemToModify.0.Transform=@n3.Transform
    Weight=@n2.OutputParam
   )
   RigUnit_BeginExecution = &RigUnit_BeginExecution()
   SequenceAggregate = &RigVMFunction_Sequence(ExecuteContext=@RigUnit_BeginExecution.ExecutePin)
   n4 = &RigUnit_SphericalPoseReader(
    ActiveRegionScaleFactors=(PositiveWidth=1,NegativeWidth=1,PositiveHeight=1,NegativeHeight=1)
    ActiveRegionSize=0.1
    Debug=(bDrawDebug=true,DebugScale=25,DebugSegments=20,DebugThickness=0.25)
    DriverAxis=(X=1,Y=0,Z=0)
    DriverItem=(Type=Bone,Name="thigh_r")
    FalloffRegionScaleFactors=(PositiveWidth=1.5,NegativeWidth=2,PositiveHeight=1.5,NegativeHeight=1.5)
    FalloffSize=0.3
    OptionalParentItem=(Type=Bone)
    RotationOffset=(X=-80,Y=0,Z=130)
    ExecutePin=@SequenceAggregate.A
   )
   n5 = &RigVMFunction_MathTransformFromSRT(Location=(X=2,Y=2,Z=20), Rotation=(X=0,Y=0,Z=0), Scale=(X=1,Y=1,Z=1))
   > /Script/ControlRig.RigUnit_SetRelativeTransformForItem(
    Child=(Type=Bone,Name="belt_corrective_01")
    Parent=(Type=Bone,Name="belt_corrective_01")
    bParentInitial=False
    bPropagateToChildren=True
    Value=@n5.Transform
    Weight=@n4.OutputParam
   ) {
    is=struct RigVMUnitNode
    template="Set Relative Transform::Execute(in Child,in Parent,in bParentInitial,in Value,in Weight,in bPropagateToChildren)"
   }
   > n6 = &RigUnit_SphericalPoseReader(
    ActiveRegionScaleFactors=(PositiveWidth=1,NegativeWidth=1,PositiveHeight=1,NegativeHeight=1)
    ActiveRegionSize=0.07
    Debug=(bDrawDebug=true,DebugScale=25,DebugSegments=20,DebugThickness=0.25)
    DriverAxis=(X=1,Y=0,Z=0)
    DriverItem=(Type=Bone,Name="thigh_r")
    FalloffRegionScaleFactors=(PositiveWidth=1.5,NegativeWidth=2,PositiveHeight=1,NegativeHeight=2.2)
    FalloffSize=0.13
    OptionalParentItem=(Type=Bone)
    RotationOffset=(X=-45,Y=0,Z=90)
   )
   n7 = &RigVMFunction_MathTransformFromSRT(Location=(X=0,Y=-5,Z=20), Rotation=(X=0,Y=0,Z=0), Scale=(X=1,Y=1,Z=1))
   > &RigUnit_SetRelativeTransformForItem(
    Child=(Type=Bone,Name="belt_corrective_01")
    Parent=(Type=Bone,Name="belt_corrective_01")
    bParentInitial=False
    bPropagateToChildren=True
    Value=@n7.Transform
    Weight=@n6.OutputParam
   ) {template="Set Relative Transform::Execute(in Child,in Parent,in bParentInitial,in Value,in Weight,in bPropagateToChildren)"}
   > n8 = &RigUnit_SphericalPoseReader(
    ActiveRegionScaleFactors=(PositiveWidth=1,NegativeWidth=1.5,PositiveHeight=2,NegativeHeight=2.3)
    ActiveRegionSize=0.05
    Debug=(bDrawDebug=true,DebugScale=25,DebugSegments=20,DebugThickness=0.25)
    DriverAxis=(X=1,Y=0,Z=0)
    DriverItem=(Type=Bone,Name="thigh_r")
    FalloffRegionScaleFactors=(PositiveWidth=2.2,NegativeWidth=1,PositiveHeight=1.5,NegativeHeight=4)
    FalloffSize=0.3
    OptionalParentItem=(Type=Bone)
    RotationOffset=(X=-70,Y=0,Z=-20)
   )
   > &RigUnit_SetRelativeTransformForItem(
    Child=(Type=Bone,Name="belt_corrective_01")
    Parent=(Type=Bone,Name="belt_corrective_01")
    bParentInitial=False
    bPropagateToChildren=True
    Value=&RigVMFunction_MathTransformFromSRT(Location=(X=10,Y=-5,Z=40), Rotation=(X=0,Y=0,Z=0), Scale=(X=1,Y=1,Z=1)).Transform
    Weight=@n8.OutputParam
   ) {template="Set Relative Transform::Execute(in Child,in Parent,in bParentInitial,in Value,in Weight,in bPropagateToChildren)"}
   > &RigUnit_SetRelativeTransformForItem(
    Child=(Type=Bone,Name="belt_corrective_02")
    Parent=(Type=Bone,Name="belt_corrective_02")
    bParentInitial=False
    bPropagateToChildren=True
    Value=&RigVMFunction_MathTransformFromSRT(Location=(X=5,Y=5,Z=7), Rotation=(X=0,Y=0,Z=0), Scale=(X=1,Y=1,Z=1)).Transform
    Weight=@n8.OutputParam
   ) {template="Set Relative Transform::Execute(in Child,in Parent,in bParentInitial,in Value,in Weight,in bPropagateToChildren)"}
   > n9 = &RigUnit_SetRelativeTransformForItem(
    Child=(Type=Bone,Name="dyn_belt")
    Parent=(Type=Bone,Name="dyn_belt")
    bParentInitial=False
    bPropagateToChildren=True
    Value=&RigVMFunction_MathTransformFromSRT(Location=(X=1,Y=-1,Z=5), Rotation=(X=40,Y=0,Z=0), Scale=(X=1,Y=1,Z=1)).Transform
    Weight=@n8.OutputParam
   ) {template="Set Relative Transform::Execute(in Child,in Parent,in bParentInitial,in Value,in Weight,in bPropagateToChildren)"}
   RigVMFunction_ControlFlowBranch = &RigVMFunction_ControlFlowBranch(ExecuteContext=@SequenceAggregate.B, Condition=RigVMVariableNode() {variable=is lod2}.Value)
   &RigUnit_ParentConstraint(
    AdvancedSettings=(RotationOrderForFilter=XZY)
    Child=(Type=Bone,Name="upperarm_twist_01_l")
    Parents=((Item=(Type=Bone,Name="upperarm_l"),Weight=1))
    Weight=1
    ExecutePin=@RigVMFunction_ControlFlowBranch.True
   )
   > &RigUnit_ParentConstraint(
    AdvancedSettings=(RotationOrderForFilter=XZY)
    Child=(Type=Bone,Name="upperarm_twist_01_r")
    Parents=((Item=(Type=Bone,Name="upperarm_r"),Weight=1))
    Weight=1
   )
   RigVMFunction_MathQuaternionToEuler = &RigVMFunction_MathQuaternionToEuler(
    RotationOrder=ZYX
    Value=&RigUnit_GetRelativeTransformForItem(Child=(Type=Bone,Name="upperarm_l"), Parent=(Type=Bone,Name="spine_01")).RelativeTransform.Rotation
   )
   > &RigUnit_ModifyTransforms(
    ItemToModify=((Item=(Type=Bone,Name="upperarm_twist_01_l"),Transform=(Rotation=(X=0,Y=0,Z=0,W=1),Translation=(X=-5,Y=0,Z=-3),Scale3D=(X=1,Y=1,Z=1))))
    ItemToModify.0.Transform=&RigVMFunction_MathTransformFromSRT(Location=(X=2,Y=0,Z=0), Rotation=(X=-50,Y=0,Z=0), Scale=(X=1,Y=1,Z=1)).Transform
    Weight=&RigVMFunction_MathDoubleRemap(SourceMaximum=-45, SourceMinimum=-40, TargetMaximum=1, TargetMinimum=0, Value=@RigVMFunction_MathQuaternionToEuler.Result.Z).Result
   )
   n10 = &RigUnit_GetRelativeTransformForItem(Child=(Type=Bone,Name="upperarm_l"), Parent=(Type=Bone,Name="spine_01"))
   > &RigUnit_ModifyTransforms(
    ItemToModify=((Item=(Type=Bone,Name="upperarm_twist_01_r"),Transform=(Rotation=(X=0,Y=0,Z=0,W=1),Translation=(X=-5,Y=0,Z=-3),Scale3D=(X=1,Y=1,Z=1))))
    ItemToModify.0.Transform=&RigVMFunction_MathTransformFromSRT(Location=(X=0,Y=0,Z=0), Rotation=(X=-60,Y=0,Z=0), Scale=(X=1,Y=1,Z=1)).Transform
    Weight=&RigVMFunction_MathDoubleRemap(
     SourceMaximum=-65
     SourceMinimum=-80
     TargetMaximum=1
     TargetMinimum=0
     Value=&RigVMFunction_MathQuaternionToEuler(RotationOrder=ZYX, Value=@n10.RelativeTransform.Rotation).Result.Z
    ).Result
   )
   &RigUnit_ModifyTransforms(
    ExecutePin=()
    ItemToModify=((Item=(Type=Bone,Name="belt_corrective_01"),Transform=(Rotation=(X=0,Y=0,Z=0,W=1),Translation=(X=0,Y=8,Z=0),Scale3D=(X=1,Y=1,Z=1))))
    ItemToModify.0.Transform=@n7.Transform
    Weight=@n6.OutputParam
   )
   &RigUnit_ModifyTransforms(
    ExecutePin=()
    ItemToModify=((Item=(Type=Bone,Name="belt_corrective_01"),Transform=(Rotation=(X=0,Y=0,Z=0,W=1),Translation=(X=0,Y=8,Z=0),Scale3D=(X=1,Y=1,Z=1))))
    ItemToModify.0.Transform=@n5.Transform
    Weight=@n4.OutputParam
   )
   &RigUnit_SetRelativeTransformForItem(
    Child=(Type=Bone,Name="dyn_belt")
    ExecutePin=()
    Parent=(Type=Bone,Name="dyn_belt")
    bParentInitial=False
    bPropagateToChildren=True
    Value=&RigVMFunction_MathTransformFromSRT(Location=(X=1,Y=0,Z=5), Rotation=(X=40,Y=0,Z=0), Scale=(X=1,Y=1,Z=1)).Transform
    Weight=@n4.OutputParam
   ) {template="Set Relative Transform::Execute(in Child,in Parent,in bParentInitial,in Value,in Weight,in bPropagateToChildren)"}
   &RigUnit_SetRelativeTransformForItem(
    Child=(Type=Bone,Name="belt_corrective_02")
    Parent=(Type=Bone,Name="belt_corrective_02")
    bParentInitial=False
    bPropagateToChildren=True
    Value=@n3.Transform
    ExecutePin=@n11.ExecutePin
    Weight=@n2.OutputParam
   ) {template="Set Relative Transform::Execute(in Child,in Parent,in bParentInitial,in Value,in Weight,in bPropagateToChildren)"}
   &RigUnit_SetRelativeTransformForItem(
    Child=(Type=Bone,Name="belt_corrective_02")
    ExecutePin=()
    Parent=(Type=Bone,Name="belt_corrective_02")
    bParentInitial=False
    bPropagateToChildren=True
    Value=&RigVMFunction_MathTransformFromSRT(Location=(X=0,Y=3,Z=10), Rotation=(X=0,Y=0,Z=0), Scale=(X=1,Y=1,Z=1)).Transform
    Weight=@n4.OutputParam
   ) {template="Set Relative Transform::Execute(in Child,in Parent,in bParentInitial,in Value,in Weight,in bPropagateToChildren)"}
   n11 = &RigUnit_SetRelativeTransformForItem(
    Child=(Type=Bone,Name="belt_corrective_01")
    Parent=(Type=Bone,Name="belt_corrective_01")
    bParentInitial=False
    bPropagateToChildren=True
    Value=@n1.Transform
    ExecutePin=@n2.ExecutePin
    Weight=@n2.OutputParam
   ) {template="Set Relative Transform::Execute(in Child,in Parent,in bParentInitial,in Value,in Weight,in bPropagateToChildren)"}
  ]}
]
hierarchy [
 root {loc=<0 0 0>, rot=<0 0 0 1>}
 pelvis {parent=root, loc=<0 -0.5301 98.0571>, rot=<0.0127 0.707 -0.0127 -0.707>}
 spine_01 {parent=pelvis, loc=<6.1208 -0.9328 0>, rot=<0 0 0.034 -0.9994>}
 spine_02 {parent=spine_01, loc=<6.5996 -0.2095 0>, rot=<0 0 0.0238 -0.9997>}
 spine_03 {parent=spine_02, loc=<6.4914 -0.1237 0>, rot=<0 0 0.0149 -0.9999>}
 spine_04 {parent=spine_03, loc=<6.9598 0.0794 0>, rot=<0 0 -0.0132 -0.9999>}
 spine_05 {parent=spine_04, loc=<6.9987 0 0>, rot=<0 0 -0.0015 -1>}
 attach_cape {parent=spine_05, loc=<8.1133 -21.5047 -0.0003>, rot=<0.0616 -0.7044 -0.0616 -0.7044>}
 neck_01 {parent=spine_05, loc=<11.3138 0.1529 0>, rot=<0 0 -0.2003 -0.9797>}
 neck_02 {parent=neck_01, loc=<4.5627 0 0>, rot=<0 0 0.0504 -0.9987>}
 head {parent=neck_02, loc=<5.3914 0 0>, rot=<0 0 0.115 -0.9934>}
 attach_backpack {parent=spine_05, loc=<-0.483 -28.8411 -0.0005>, rot=<-0.0093 -0.7071 0.0093 -0.707>}
 clavicle_l {parent=spine_05, loc=<5.4831 -0.1053 -3.2946>, rot=<0.0465 -0.7289 -0.009 -0.6829>}
 upperarm_l {parent=clavicle_l, loc=<9.8658 -1.2136 -0.4105>, rot=<0.0164 -0.3245 -0.0269 -0.9454>}
 lowerarm_l {parent=upperarm_l, loc=<26.465 -0.8345 0.259>, rot=<0.0036 0.0161 -0.2298 -0.9731>}
 hand_l {parent=lowerarm_l, loc=<22.7355 -0.3203 -0.6105>, rot=<-0.6784 -0.0223 0.0735 -0.7307>}
 thumb_01_l {parent=hand_l, loc=<1.3118 -0.3191 -2.459>, rot=<0.6546 -0.4042 0.0664 -0.6354>}
 thumb_02_l {parent=thumb_01_l, loc=<5.9336 -0.0448 0.0401>, rot=<-0.0037 0.0213 0.298 -0.9543>}
 thumb_03_l {parent=thumb_02_l, loc=<2.2223 0.3682 -0.0907>, rot=<0.0118 -0.013 0.0066 -0.9998>}
 index_metacarpal_l {parent=hand_l, loc=<1.6535 0.0183 -2.96>, rot=<0.0158 -0.0124 -0.0173 -0.9996>}
 index_01_l {parent=index_metacarpal_l, loc=<8.3008 0.0357 -0.0524>, rot=<-0.0044 0.0485 0.0924 -0.9945>}
 index_02_l {parent=index_01_l, loc=<3.3587 -0.3004 0.0452>, rot=<-0.002 0.0027 0.3225 -0.9466>}
 index_03_l {parent=index_02_l, loc=<2.5578 -0.123 -0.0252>, rot=<-0.002 0.0067 0.2558 -0.9667>}
 middle_metacarpal_l {parent=hand_l, loc=<2.1856 -0.1725 -0.5984>, rot=<0.006 0.0168 -0.0329 -0.9993>}
 middle_01_l {parent=middle_metacarpal_l, loc=<7.3275 -0.2447 0.0129>, rot=<-0.0208 0.0042 0.0858 -0.9961>}
 middle_02_l {parent=middle_01_l, loc=<4.0123 0.1401 0.1175>, rot=<-0.0218 0.0258 0.3512 -0.9357>}
 middle_03_l {parent=middle_02_l, loc=<2.6077 0.0534 0.1373>, rot=<0.0082 0.0188 0.1708 -0.9851>}
 ring_metacarpal_l {parent=hand_l, loc=<2.2074 -0.3625 1.3638>, rot=<-0.0284 0.0722 -0.0458 -0.9959>}
 ring_01_l {parent=ring_metacarpal_l, loc=<7.0295 -0.6197 -0.1446>, rot=<-0.0489 -0.0252 0.1892 -0.9804>}
 ring_02_l {parent=ring_01_l, loc=<3.6174 0.4647 -0.1027>, rot=<0 0 0.2806 -0.9598>}
 ring_03_l {parent=ring_02_l, loc=<2.4152 0.114 0.0199>, rot=<0 0 0.1189 -0.9929>}
 pinky_metacarpal_l {parent=hand_l, loc=<1.7616 -0.9753 3.3814>, rot=<-0.0397 0.0814 -0.069 -0.9935>}
 pinky_01_l {parent=pinky_metacarpal_l, loc=<7.1709 -0.4831 -0.1094>, rot=<-0.0654 -0.0035 0.2242 -0.9723>}
 pinky_02_l {parent=pinky_01_l, loc=<2.8274 0.0893 -0.1043>, rot=<0.0037 0.0011 0.2854 -0.9584>}
 pinky_03_l {parent=pinky_02_l, loc=<2.2046 0.0307 -0.0078>, rot=<-0.0002 0.0098 0.1445 -0.9895>}
 deform_wrist_dn_base_l {parent=hand_l, loc=<0 -1.5326 0>, rot=<0 0 0 -1>}
 deform_wrist_dn_l {parent=deform_wrist_dn_base_l, loc=<0 0 0>, rot=<0 0 0 1>}
 deform_wrist_up_base_l {parent=hand_l, loc=<0 2.3009 0>, rot=<0 0 0 -1>}
 deform_wrist_up_l {parent=deform_wrist_up_base_l, loc=<0 0 0>, rot=<0 0 0 1>}
 lowerarm_twist_02_l {parent=lowerarm_l, loc=<10.6707 0 0>, rot=<0 0 0 1>}
 lowerarm_twist_01_l {parent=lowerarm_l, loc=<18.316 0 0>, rot=<0 0 0 1>}
 deform_elbow_in_base_l {parent=lowerarm_l, loc=<0.811 3.4798 -0.2543>, rot=<-0.0001 0.0017 0.0887 -0.9961>}
 deform_elbow_in_l {parent=deform_elbow_in_base_l, loc=<0 0 0>, rot=<0 0 0 1>}
 deform_elbow_out_base_l {parent=lowerarm_l, loc=<0 -4 0>, rot=<0 0 0 1>}
 deform_elbow_out_l {parent=deform_elbow_out_base_l, loc=<0 0 0>, rot=<0 0 0 1>}
 upperarm_twist_01_l {parent=upperarm_l, loc=<7.2135 -0.1134 0>, rot=<0 0 0 1>}
 deform_upperarm_bk_base_l {parent=upperarm_twist_01_l, loc=<-6.4053 -4.3746 0.314>, rot=<0 0 0 -1>}
 deform_upperarm_bk_l {parent=deform_upperarm_bk_base_l, loc=<0 0 0>, rot=<0 0 0 1>}
 deform_upperarm_fr_base_l {parent=upperarm_twist_01_l, loc=<-5.6312 5.092 0.3354>, rot=<0 0 0 -1>}
 deform_upperarm_fr_l {parent=deform_upperarm_fr_base_l, loc=<0 0 0>, rot=<0 0 0 1>}
 deform_upperarm_up_base_l {parent=upperarm_twist_01_l, loc=<-8.891 0.4825 3.9517>, rot=<0 0 0 -1>}
 deform_upperarm_up_l {parent=deform_upperarm_up_base_l, loc=<0 0 0>, rot=<0 0 0 1>}
 sleeve_l {parent=upperarm_twist_01_l, loc=<0 0 0>, rot=<0 0 0 1>}
 dyn_sleeve_l_09 {parent=sleeve_l, loc=<-2.2309 -0.1909 -0.013>, rot=<-0.8079 0.0278 -0.0507 -0.5865>}
 dyn_sleeve_l_12 {parent=sleeve_l, loc=<-2.2309 -0.1909 -0.013>, rot=<0.9768 0.03 0.004 -0.212>}
 dyn_sleeve_l_11 {parent=sleeve_l, loc=<-2.2309 -0.1909 -0.013>, rot=<0.5435 0.0241 0.0812 -0.8352>}
 dyn_sleeve_l_10 {parent=sleeve_l, loc=<-2.2309 -0.1909 -0.013>, rot=<-0.0013 0.0511 0.0114 -0.9986>}
 upperarm_twist_02_l {parent=upperarm_l, loc=<19.7998 -0.1134 0>, rot=<0 0 0 1>}
 clavicle_r {parent=spine_05, loc=<5.4831 -0.1053 3.2946>, rot=<0.7289 0.0465 -0.6829 0.009>}
 upperarm_r {parent=clavicle_r, loc=<-9.8658 1.2136 0.4105>, rot=<0.0164 -0.3245 -0.0269 -0.9454>}
 lowerarm_r {parent=upperarm_r, loc=<-26.465 0.8345 -0.259>, rot=<0.0036 0.0161 -0.2298 -0.9731>}
 hand_r {parent=lowerarm_r, loc=<-22.7355 0.3203 0.6106>, rot=<-0.6784 -0.0223 0.0735 -0.7307>}
 thumb_01_r {parent=hand_r, loc=<-1.3118 0.3191 2.459>, rot=<0.6546 -0.4042 0.0664 -0.6354>}
 thumb_02_r {parent=thumb_01_r, loc=<-5.9336 0.0448 -0.0401>, rot=<-0.0037 0.0213 0.298 -0.9543>}
 thumb_03_r {parent=thumb_02_r, loc=<-2.2223 -0.3682 0.0907>, rot=<0.0118 -0.013 0.0066 -0.9998>}
 index_metacarpal_r {parent=hand_r, loc=<-1.6535 -0.0183 2.96>, rot=<0.0158 -0.0124 -0.0173 -0.9996>}
 index_01_r {parent=index_metacarpal_r, loc=<-8.3008 -0.0357 0.0524>, rot=<-0.0044 0.0485 0.0924 -0.9945>}
 index_02_r {parent=index_01_r, loc=<-3.3587 0.3004 -0.0452>, rot=<-0.002 0.0027 0.3225 -0.9466>}
 index_03_r {parent=index_02_r, loc=<-2.5578 0.123 0.0252>, rot=<-0.002 0.0067 0.2558 -0.9667>}
 middle_metacarpal_r {parent=hand_r, loc=<-2.1856 0.1725 0.5984>, rot=<0.006 0.0168 -0.0329 -0.9993>}
 middle_01_r {parent=middle_metacarpal_r, loc=<-7.3275 0.2447 -0.0129>, rot=<-0.0208 0.0042 0.0858 -0.9961>}
 middle_02_r {parent=middle_01_r, loc=<-4.0123 -0.1401 -0.1175>, rot=<-0.0218 0.0258 0.3512 -0.9357>}
 middle_03_r {parent=middle_02_r, loc=<-2.6077 -0.0534 -0.1373>, rot=<0.0082 0.0188 0.1708 -0.9851>}
 ring_metacarpal_r {parent=hand_r, loc=<-2.2074 0.3625 -1.3638>, rot=<-0.0284 0.0722 -0.0458 -0.9959>}
 ring_01_r {parent=ring_metacarpal_r, loc=<-7.0295 0.6197 0.1446>, rot=<-0.0489 -0.0252 0.1892 -0.9804>}
 ring_02_r {parent=ring_01_r, loc=<-3.6174 -0.4647 0.1027>, rot=<0 0 0.2806 -0.9598>}
 ring_03_r {parent=ring_02_r, loc=<-2.4152 -0.114 -0.0199>, rot=<0 0 0.1189 -0.9929>}
 pinky_metacarpal_r {parent=hand_r, loc=<-1.7616 0.9753 -3.3814>, rot=<-0.0397 0.0814 -0.069 -0.9935>}
 pinky_01_r {parent=pinky_metacarpal_r, loc=<-7.1709 0.4831 0.1094>, rot=<-0.0654 -0.0035 0.2242 -0.9723>}
 pinky_02_r {parent=pinky_01_r, loc=<-2.8274 -0.0893 0.1043>, rot=<0.0037 0.0011 0.2854 -0.9584>}
 pinky_03_r {parent=pinky_02_r, loc=<-2.2046 -0.0307 0.0078>, rot=<-0.0002 0.0098 0.1445 -0.9895>}
 deform_wrist_dn_base_r {parent=hand_r, loc=<-0.0001 1.5328 0>, rot=<0 0 0 -1>}
 deform_wrist_dn_r {parent=deform_wrist_dn_base_r, loc=<0 0 0>, rot=<0 0 0 1>}
 deform_wrist_up_base_r {parent=hand_r, loc=<-0.0002 -2.3007 -0.0001>, rot=<0 0 0 -1>}
 deform_wrist_up_r {parent=deform_wrist_up_base_r, loc=<0 0 0>, rot=<0 0 0 1>}
 lowerarm_twist_02_r {parent=lowerarm_r, loc=<-10.6707 0 0>, rot=<0 0 0 -1>}
 lowerarm_twist_01_r {parent=lowerarm_r, loc=<-18.316 0 0>, rot=<0 0 0 -1>}
 deform_elbow_in_base_r {parent=lowerarm_r, loc=<-0.1834 -3.6062 0.2572>, rot=<0 0 0 -1>}
 deform_elbow_in_r {parent=deform_elbow_in_base_r, loc=<0 0 0>, rot=<0 0 0 1>}
 deform_elbow_out_base_r {parent=lowerarm_r, loc=<0 4 -0.0001>, rot=<0 0 0 -1>}
 deform_elbow_out_r {parent=deform_elbow_out_base_r, loc=<0 0 0>, rot=<0 0 0 1>}
 upperarm_twist_01_r {parent=upperarm_r, loc=<-7.2135 0.1134 0>, rot=<0 0 0 -1>}
 deform_upperarm_bk_base_r {parent=upperarm_twist_01_r, loc=<6.4053 4.3746 -0.314>, rot=<0 0 0 -1>}
 deform_upperarm_bk_r {parent=deform_upperarm_bk_base_r, loc=<0 0 0>, rot=<0 0 0 1>}
 deform_upperarm_fr_base_r {parent=upperarm_twist_01_r, loc=<5.6312 -5.092 -0.3354>, rot=<0 0 0 -1>}
 deform_upperarm_fr_r {parent=deform_upperarm_fr_base_r, loc=<0 0 0>, rot=<0 0 0 1>}
 deform_upperarm_up_base_r {parent=upperarm_twist_01_r, loc=<8.891 -0.4825 -3.9517>, rot=<0 0 0 -1>}
 deform_upperarm_up_r {parent=deform_upperarm_up_base_r, loc=<0 0 0>, rot=<0 0 0 1>}
 sleeve_r {parent=upperarm_twist_01_r, loc=<0 0 0>, rot=<0 0 0 1>}
 dyn_sleeve_r_09 {parent=sleeve_r, loc=<2.6918 -0.1573 0.1121>, rot=<-0.0232 -0.6611 0.749 -0.0374>}
 dyn_sleeve_r_10 {parent=sleeve_r, loc=<2.6918 -0.1573 0.1121>, rot=<-0.0056 -0.1641 0.9841 0.0673>}
 dyn_sleeve_r_12 {parent=sleeve_r, loc=<2.6918 -0.1573 0.1121>, rot=<-0.0362 0.7036 0.7074 0.0562>}
 dyn_sleeve_r_11 {parent=sleeve_r, loc=<2.6918 -0.1573 0.1121>, rot=<-0.0427 -0.9873 0.1435 -0.0528>}
 upperarm_twist_02_r {parent=upperarm_r, loc=<-19.7998 0.1134 0>, rot=<0 0 0 -1>}
 dyn_shirt_l_01 {parent=spine_04, loc=<-1.4422 10.9996 -7.6359>, rot=<0.0261 -0.9277 -0.3721 -0.0159>}
 dyn_shirt_l_02 {parent=spine_04, loc=<-1.7303 5.1487 -11.0394>, rot=<0.0394 -0.9967 -0.0642 -0.0311>}
 dyn_shirt_l_03 {parent=spine_04, loc=<-1.6418 -1.0026 -11.5029>, rot=<0.0435 -0.9968 0.057 -0.0361>}
 dyn_shirt_l_04 {parent=spine_04, loc=<-0.9111 -6.0595 -8.4411>, rot=<0.0523 -0.8937 0.443 -0.0487>}
 dyn_shirt_r_04 {parent=spine_04, loc=<-0.9217 -6.0516 8.7966>, rot=<0.049 -0.4532 0.8885 -0.0524>}
 dyn_shirt_r_03 {parent=spine_04, loc=<-1.6603 -1.2452 11.2555>, rot=<0.0341 -0.006 0.9985 -0.0418>}
 dyn_shirt_r_02 {parent=spine_04, loc=<-1.936 4.7475 10.9121>, rot=<0.0259 0.1782 0.983 -0.0349>}
 dyn_shirt_r_01 {parent=spine_04, loc=<-1.8462 10.5894 7.4877>, rot=<0.0137 0.4122 0.9107 -0.024>}
 dyn_shirt_mid {parent=spine_04, loc=<-1.1304 12.6921 -0.3407>, rot=<-0.0114 0.6858 0.7277 0>}
 dyn_shirt_back {parent=spine_04, loc=<-0.4716 -7.3662 -0.2406>, rot=<0.0039 -0.6937 0.7203 -0.0022>}
 deform_pec_aim_base_l {parent=spine_04, loc=<3.2456 9.3309 -1.2237>, rot=<0.0603 -0.4033 0.3318 -0.8507>}
 deform_pec_aim_l {parent=deform_pec_aim_base_l, loc=<0 0 0>, rot=<0 0 0 1>}
 deform_pec_base_l {parent=deform_pec_aim_l, loc=<9.3665 0 0>, rot=<0 0 0 1>}
 deform_pec_l {parent=deform_pec_base_l, loc=<0 0 0>, rot=<0 0 0 1>}
 deform_pec_aim_base_r {parent=spine_04, loc=<3.2452 9.3309 1.2237>, rot=<-0.4033 -0.0603 0.8507 0.3318>}
 deform_pec_aim_r {parent=deform_pec_aim_base_r, loc=<0 0 0>, rot=<0 0 0 1>}
 deform_pec_base_r {parent=deform_pec_aim_r, loc=<-9.3667 -0.0003 -0.0004>, rot=<0 0 0 1>}
 deform_pec_r {parent=deform_pec_base_r, loc=<0 0 0>, rot=<0 0 0 1>}
 thigh_l {parent=pelvis, loc=<-1.0855 -0.2109 -8.7053>, rot=<0.0009 0.0355 0.0155 -0.9993>}
 calf_l {parent=thigh_l, loc=<-42.5284 -0.0019 -0.7297>, rot=<0.0006 -0.0161 -0.0391 -0.9991>}
 foot_l {parent=calf_l, loc=<-46.7401 -2.5014 -0.592>, rot=<0.0427 -0.0238 0.0353 -0.9982>}
 ball_l {parent=foot_l, loc=<-3.5658 10.8066 -0.1544>, rot=<0.0001 -0.0001 -0.7063 -0.7079>}
 lace_base_l {parent=foot_l, loc=<2.5356 7.0982 0.4957>, rot=<0.2539 -0.6115 0.3021 -0.6858>}
 lace_l_01 {parent=lace_base_l, loc=<0.5082 0.2791 -0.0314>, rot=<0.5474 0.1612 -0.3605 -0.7378>}
 lace_l_02 {parent=lace_base_l, loc=<-0.1865 -0.4289 0.1232>, rot=<0.0945 -0.1167 0.8766 -0.4571>}
 calf_twist_02_l {parent=calf_l, loc=<-19.1546 -1.7943 -0.2638>, rot=<0.0237 0.0095 -0.0035 -0.9997>}
 calf_twist_01_l {parent=calf_l, loc=<-35.8483 -1.9197 -0.5794>, rot=<0.0237 0.0095 -0.0035 -0.9997>}
 dyn_shorts_l_04 {parent=calf_l, loc=<0.0393 -0.1057 -4.3331>, rot=<0.7203 0.0261 0.6891 0.0756>}
 dyn_shorts_l_01 {parent=calf_l, loc=<-1.2249 4.8452 2.1361>, rot=<-0.0309 -0.7201 0.0276 -0.6927>}
 dyn_shorts_l_03 {parent=calf_l, loc=<-0.2196 -5.9291 1.4094>, rot=<-0.5157 -0.5035 -0.4594 -0.5192>}
 dyn_shorts_l_05 {parent=calf_l, loc=<-0.4702 0.4336 5.3345>, rot=<0.0981 -0.714 0.1442 -0.6781>}
 deform_knee_in_base_l {parent=calf_l, loc=<0.8031 -5.0231 0.2456>, rot=<0 0 0 -1>}
 deform_knee_in_l {parent=deform_knee_in_base_l, loc=<0 0 0>, rot=<0 0 0 1>}
 deform_knee_out_base_l {parent=calf_l, loc=<0 4 0>, rot=<0 0 0 -1>}
 deform_knee_out_l {parent=deform_knee_out_base_l, loc=<0 0 0>, rot=<0 0 0 1>}
 thigh_twist_01_l {parent=thigh_l, loc=<-27.952 0 0>, rot=<0.0013 0 0 -1>}
 dyn_base_flask {parent=thigh_twist_01_l, loc=<22.5293 0.4485 -8.1259>, rot=<-0.0031 -0.7327 0.0009 -0.6805>}
 dyn_flask {parent=dyn_base_flask, loc=<0 0 0>, rot=<0 0 0 1>}
 deform_groin_base_l {parent=thigh_l, loc=<0.6223 5.66 -0.2386>, rot=<-0.0006 0.0022 0.0125 -0.9999>}
 deform_groin_l {parent=deform_groin_base_l, loc=<0 0 0>, rot=<0 0 0 1>}
 thigh_r {parent=pelvis, loc=<-1.0855 -0.2109 8.7053>, rot=<0.0355 -0.0009 0.9993 0.0155>}
 calf_r {parent=thigh_r, loc=<42.5284 0.0019 0.7297>, rot=<0.0006 -0.0161 -0.0391 -0.9991>}
 foot_r {parent=calf_r, loc=<46.7401 2.5014 0.592>, rot=<0.0427 -0.0238 0.0353 -0.9982>}
 ball_r {parent=foot_r, loc=<3.5658 -10.8066 0.1544>, rot=<0.0001 -0.0001 -0.7063 -0.7079>}
 lace_base_r {parent=foot_r, loc=<-2.5353 -7.0439 -0.443>, rot=<0.2539 -0.6115 0.3021 -0.6858>}
 lace_r_01 {parent=lace_base_r, loc=<-0.4787 -0.2139 0.0552>, rot=<0.2742 -0.748 0.5882 -0.1392>}
 lace_r_02 {parent=lace_base_r, loc=<0.262 0.4294 -0.1289>, rot=<-0.1567 -0.137 -0.4284 -0.8793>}
 calf_twist_02_r {parent=calf_r, loc=<19.1546 1.7943 0.2638>, rot=<0.0237 0.0095 -0.0035 -0.9997>}
 calf_twist_01_r {parent=calf_r, loc=<35.8483 1.9197 0.5794>, rot=<0.0237 0.0095 -0.0035 -0.9997>}
 dyn_shorts_r_02 {parent=calf_r, loc=<0.4768 -0.354 -5.3342>, rot=<-0.0275 -0.7202 0.0241 -0.6928>}
 dyn_shorts_r_03 {parent=calf_r, loc=<0.2196 5.9291 -1.4094>, rot=<-0.481 -0.5399 -0.4956 -0.4813>}
 dyn_shorts_r_04 {parent=calf_r, loc=<-0.0393 0.1057 4.3331>, rot=<0.7203 0.0261 0.6891 0.0756>}
 dyn_shorts_r_06 {parent=calf_r, loc=<1.2304 -4.8988 -1.8811>, rot=<0.3281 -0.6068 0.4683 -0.5521>}
 deform_knee_in_base_r {parent=calf_r, loc=<-0.8031 5.0231 -0.2457>, rot=<0 0 0 -1>}
 deform_knee_in_r {parent=deform_knee_in_base_r, loc=<0 0 0>, rot=<0 0 0 1>}
 deform_knee_out_base_r {parent=calf_r, loc=<0 -4 0>, rot=<0 0 0 -1>}
 deform_knee_out_r {parent=deform_knee_out_base_r, loc=<0 0 0>, rot=<0 0 0 1>}
 thigh_twist_01_r {parent=thigh_r, loc=<27.952 0 0>, rot=<0.0013 0 0 -1>}
 deform_groin_base_r {parent=thigh_r, loc=<-0.4801 -5.6776 0.2486>, rot=<0 0 0 -1>}
 deform_groin_r {parent=deform_groin_base_r, loc=<0 0 0>, rot=<0 0 0 1>}
 belt_corrective_01 {parent=pelvis, loc=<-5.541 0.3415 17.1359>, rot=<-0.0127 -0.707 0.0127 -0.707>}
 belt_corrective_02 {parent=pelvis, loc=<-1.6419 7.3968 9.0852>, rot=<-0.0127 -0.707 0.0127 -0.707>}
 dyn_base_belt {parent=pelvis, loc=<4.99 9.8059 -1.8405>, rot=<-0.494 -0.506 -0.4755 -0.5233>}
 dyn_belt {parent=dyn_base_belt, loc=<0.0001 -0.0014 -0.0007>, rot=<0 0 0 1>}
 deform_glute_base_l {parent=pelvis, loc=<-7.3466 -9.0213 -9.1245>, rot=<0.0172 0.031 -0.4477 -0.8935>}
 deform_glute_l {parent=deform_glute_base_l, loc=<0 0 0>, rot=<0 0 0 1>}
 deform_glute_base_r {parent=pelvis, loc=<-7.3464 -9.0213 9.1245>, rot=<0.031 -0.0172 0.8935 -0.4477>}
 deform_glute_r {parent=deform_glute_base_r, loc=<0 0 0>, rot=<0 0 0 1>}
]
variables [is lod2 {type=bool, default=False}]
