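# Physics asset for the shirt mesh of the female medium BrightShimmer Casual body cosmetic, defining collision and joints for its dynamic shirt bones.
# Small, identically sized capsules sit on ten shirt bones spanning left, right, mid and back, with a larger capsule on spine_04.
# Each shirt bone constrains directly to spine_04 rather than chaining, with identical narrow swing limits and zero twist limit.
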PhysicsAsset /BRCosmetics/Characters/Player/Female/Medium/Bodies/F_MED_BrightShimmer_Casual/Meshes/F_MED_BrightShimmer_Casual_Shirt_Physics
bodies {
 dyn_shirt_back [{shapeType=Capsule, length=1.5, radius=1.5, shapeName=dyn_shirt_back_capsule, center=<4.5657 -0.0814 0.0884>, rotation=<rot 90 0 0>}]
 dyn_shirt_l_01 [{shapeType=Capsule, length=1.5, radius=1.5, shapeName=dyn_shirt_l_01_capsule, center=<4.0315 -1.0285 -0.7375>, rotation=<rot 90 0 0>}]
 dyn_shirt_l_02 [{shapeType=Capsule, length=1.5, radius=1.5, shapeName=dyn_shirt_l_02_capsule, center=<3.8302 -0.2167 -0.6483>, rotation=<rot 90 0 0>}]
 dyn_shirt_l_03 [{shapeType=Capsule, length=1.5, radius=1.5, shapeName=dyn_shirt_l_03_capsule, center=<3.8627 0.4419 -0.6075>, rotation=<rot 90 0 0>}]
 dyn_shirt_l_04 [{shapeType=Capsule, length=1.5, radius=1.5, shapeName=dyn_shirt_l_04_capsule, center=<4.1961 0.5568 -0.9843>, rotation=<rot 90 0 0>}]
 dyn_shirt_mid [{shapeType=Capsule, length=1.5, radius=1.5, shapeName=dyn_shirt_mid_capsule, center=<4.0057 0.2034 -0.2531>, rotation=<rot 90 0 0>}]
 dyn_shirt_r_01 [{shapeType=Capsule, length=1.5, radius=1.5, shapeName=dyn_shirt_r_01_capsule, center=<3.7511 -0.082 -0.7704>, rotation=<rot 90 0 0>}]
 dyn_shirt_r_02 [{shapeType=Capsule, length=1.5, radius=1.5, shapeName=dyn_shirt_r_02_capsule, center=<3.8968 -0.2192 -0.6512>, rotation=<rot 90 0 0>}]
 dyn_shirt_r_03 [{shapeType=Capsule, length=1.5, radius=1.5, shapeName=dyn_shirt_r_03_capsule, center=<3.9236 -0.1269 -0.6938>, rotation=<rot 90 0 0>}]
 dyn_shirt_r_04 [{shapeType=Capsule, length=1.5, radius=1.5, shapeName=dyn_shirt_r_04_capsule, center=<4.0496 -0.0854 -1.2317>, rotation=<rot 90 0 0>}]
 spine_04 [{
   shapeType=Capsule
   length=6.5828
   radius=7.4425
   shapeName=spine_04_capsule
   center=<-4.2074 1.2661 -0.1572>
   rotation=<rot 0.1091 -90.0002 179.7396>
  }]
}
constraints [
 {bone1Name=dyn_shirt_r_01, bone2Name=spine_04, swing1LimitDegrees=20, swing2LimitDegrees=5, twistLimitDegrees=0}
 {bone1Name=dyn_shirt_mid, bone2Name=spine_04, swing1LimitDegrees=20, swing2LimitDegrees=5, twistLimitDegrees=0}
 {bone1Name=dyn_shirt_l_01, bone2Name=spine_04, swing1LimitDegrees=20, swing2LimitDegrees=5, twistLimitDegrees=0}
 {bone1Name=dyn_shirt_l_02, bone2Name=spine_04, swing1LimitDegrees=20, swing2LimitDegrees=5, twistLimitDegrees=0}
 {bone1Name=dyn_shirt_l_03, bone2Name=spine_04, swing1LimitDegrees=20, swing2LimitDegrees=5, twistLimitDegrees=0}
 {bone1Name=dyn_shirt_l_04, bone2Name=spine_04, swing1LimitDegrees=20, swing2LimitDegrees=5, twistLimitDegrees=0}
 {bone1Name=dyn_shirt_r_02, bone2Name=spine_04, swing1LimitDegrees=20, swing2LimitDegrees=5, twistLimitDegrees=0}
 {bone1Name=dyn_shirt_r_03, bone2Name=spine_04, swing1LimitDegrees=20, swing2LimitDegrees=5, twistLimitDegrees=0}
 {bone1Name=dyn_shirt_r_04, bone2Name=spine_04, swing1LimitDegrees=20, swing2LimitDegrees=5, twistLimitDegrees=0}
 {bone1Name=dyn_shirt_back, bone2Name=spine_04, swing1LimitDegrees=20, swing2LimitDegrees=5, twistLimitDegrees=0}
]
