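# Animation Blueprint for the BrightShimmer Casual female body cosmetic, adding secondary motion to the pose copied from its parent mesh.
# An LOD-aware Control Rig feeds a collar offset, shirt RigidBody, AnimDynamics for flask, shorts, sleeves and laces, then a deform Control Rig.
# The event graph drives World Position Offset cloth parameters on body and overlay materials from calf transforms, enabling wind and turbulence during skydiving or parachuting.
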AnimBlueprint /BRCosmetics/Characters/Player/Female/Medium/Bodies/F_MED_BrightShimmer_Casual/Meshes/F_MED_BrightShimmer_Casual_AnimBP
target_skeleton=&/F_MED_BrightShimmer_Casual_Skeleton
graphs [
 AnimGraph {nodes [
   n1 = Control Rig!(Source=Copy Pose From Mesh() {bUseAttachedParent=true}, Alpha, is lod2:bool=Get Is lod2() -> (Is lod2:bool)) -> (Pose) {
    alphaBoolBlend {blendOption=Linear}
    controlRigAssetReference {blueprintRigClass=&/F_MED_BrightShimmer_Casual_CtrlRig.F_MED_BrightShimmer_Casual_CtrlRig_C}
   }
   n2 = "Transform (Modify) Bone - Bone: collar_corrective"(ComponentPose=Local To Component(LocalPose=@n1), Alpha=1, Translation="0,1,0", Rotation="0,0,0", Scale="1,1,1") {
    translationMode=BMM_Additive
    translationSpace=BCS_BoneSpace
    alphaBoolBlend {blendOption=Linear}
    boneToModify {boneName=collar_corrective}
    translation {y=1}
   }
   n3 = RigidBody(ComponentPose=@n2) {
    bOverrideWorldGravity=true
    overridePhysicsAsset=&/F_MED_BrightShimmer_Casual_Shirt_Physics
    simulationSpace=BaseBoneSpace
    alphaBoolBlend {blendOption=Linear}
    baseBoneRef {boneName=spine_03}
    componentAppliedLinearAccClamp=<2500 2500 2500>
    componentLinearAccScale=<0.1 0.1 0.1>
    overrideWorldGravity {z=-980}
    simSpaceSettings {
     maxAngularAcceleration=250
     maxAngularVelocity=3
     maxLinearAcceleration=200
     maxLinearVelocity=250
     velocityScaleZ=0
     worldAlpha=1
     externalLinearDragV=<0.1 0.1>
    }
   }
   n4 = Anim Dynamics - Bone: dyn_flask(ComponentPose=@n3, Alpha=1, ExternalForce="0,0,0") {
    angularDampingOverride=1
    angularSpringConstant=5000
    bAngularSpring=true
    bGravityOverrideInSimSpace=true
    bOverrideAngularDamping=true
    gravityScale=0.35
    simulationSpace=RootRelative
    alphaBoolBlend {blendOption=Linear}
    alphaScaleBiasClamp {scale=0.2}
    boundBone {boneName=dyn_flask}
    componentLinearAccScale {z=0.6}
    componentLinearVelScale {z=0.6}
    physicsBodyDefinitions [{
      sphereCollisionRadius=0
      boundBone {boneName=dyn_flask}
      boxExtents=<10 20 20>
      constraintSetup {angularLimitsMax=<45 0 0>, angularLimitsMin=<-45 0 0>, angularTarget=<1 0 0>}
      localJointOffset=<7 0 0>
     }]
   }
   n5 = Anim Dynamics - Bone: dyn_shorts_l_01(ComponentPose=@n4, Alpha=1, ExternalForce="0,0,0") {
    alpha=0.9
    angularSpringConstant=1000
    bAngularSpring=true
    bGravityOverrideInSimSpace=true
    numSolverIterationsPostUpdate=4
    numSolverIterationsPreUpdate=16
    simulationSpace=Actor
    alphaBoolBlend {blendOption=Linear}
    boundBone {boneName=dyn_shorts_l_01}
    componentAppliedLinearAccClamp=<1000 1000 1000>
    componentLinearAccScale=<1 1 1>
    componentLinearVelScale=<1 1 1>
    physicsBodyDefinitions [{
      sphereCollisionRadius=0
      boundBone {boneName=dyn_shorts_l_01}
      boxExtents=<3 3 3>
      constraintSetup {angularTargetAxis=AxisZ, twistAxis=AxisY, angularLimitsMax=<0 0 30>, angularLimitsMin=<0 0 -30>, angularTarget=<0 0 1>}
      localJointOffset=<0 0 0>
     }]
    retargetingSettings {bUseAbsoluteAngle=true}
   }
   n6 = Anim Dynamics - Bone: dyn_shorts_l_03(ComponentPose=@n5, Alpha=1, ExternalForce="0,0,0") {
    alpha=0.9
    angularSpringConstant=1000
    bAngularSpring=true
    bGravityOverrideInSimSpace=true
    numSolverIterationsPostUpdate=4
    numSolverIterationsPreUpdate=16
    simulationSpace=Actor
    alphaBoolBlend {blendOption=Linear}
    boundBone {boneName=dyn_shorts_l_03}
    componentAppliedLinearAccClamp=<1000 1000 1000>
    componentLinearAccScale=<1 1 1>
    componentLinearVelScale=<1 1 1>
    physicsBodyDefinitions [{
      sphereCollisionRadius=0
      boundBone {boneName=dyn_shorts_l_03}
      boxExtents=<3 3 3>
      constraintSetup {angularTargetAxis=AxisZ, twistAxis=AxisY, angularLimitsMax=<0 0 30>, angularLimitsMin=<0 0 -30>, angularTarget=<0 0 1>}
      localJointOffset=<0 0 0>
     }]
    retargetingSettings {bUseAbsoluteAngle=true}
   }
   n7 = Anim Dynamics - Bone: dyn_shorts_l_04(ComponentPose=@n6, Alpha=1, ExternalForce="0,0,0") {
    alpha=0.9
    angularSpringConstant=1000
    bAngularSpring=true
    bGravityOverrideInSimSpace=true
    numSolverIterationsPostUpdate=4
    numSolverIterationsPreUpdate=16
    simulationSpace=Actor
    alphaBoolBlend {blendOption=Linear}
    boundBone {boneName=dyn_shorts_l_04}
    componentAppliedLinearAccClamp=<1000 1000 1000>
    componentLinearAccScale=<1 1 1>
    componentLinearVelScale=<1 1 1>
    physicsBodyDefinitions [{
      sphereCollisionRadius=0
      boundBone {boneName=dyn_shorts_l_04}
      boxExtents=<3 3 3>
      constraintSetup {angularTargetAxis=AxisZ, twistAxis=AxisY, angularLimitsMax=<0 0 30>, angularLimitsMin=<0 0 -30>, angularTarget=<0 0 1>}
      localJointOffset=<0 0 0>
     }]
    retargetingSettings {bUseAbsoluteAngle=true}
   }
   n8 = Anim Dynamics - Bone: dyn_shorts_l_05(ComponentPose=@n7, Alpha=1, ExternalForce="0,0,0") {
    alpha=0.9
    angularSpringConstant=1000
    bAngularSpring=true
    bGravityOverrideInSimSpace=true
    numSolverIterationsPostUpdate=4
    numSolverIterationsPreUpdate=16
    simulationSpace=Actor
    alphaBoolBlend {blendOption=Linear}
    boundBone {boneName=dyn_shorts_l_05}
    componentAppliedLinearAccClamp=<1000 1000 1000>
    componentLinearAccScale=<1 1 1>
    componentLinearVelScale=<1 1 1>
    physicsBodyDefinitions [{
      sphereCollisionRadius=0
      boundBone {boneName=dyn_shorts_l_05}
      boxExtents=<3 3 3>
      constraintSetup {angularTargetAxis=AxisZ, twistAxis=AxisY, angularLimitsMax=<0 0 30>, angularLimitsMin=<0 0 -30>, angularTarget=<0 0 1>}
      localJointOffset=<0 0 0>
     }]
    retargetingSettings {bUseAbsoluteAngle=true}
   }
   n9 = Anim Dynamics - Bone: dyn_shorts_r_02(ComponentPose=@n8, Alpha=1, ExternalForce="0,0,0") {
    alpha=0.9
    angularSpringConstant=1000
    bAngularSpring=true
    bGravityOverrideInSimSpace=true
    numSolverIterationsPostUpdate=4
    numSolverIterationsPreUpdate=16
    simulationSpace=Actor
    alphaBoolBlend {blendOption=Linear}
    boundBone {boneName=dyn_shorts_r_02}
    componentAppliedLinearAccClamp=<1000 1000 1000>
    componentLinearAccScale=<1 1 1>
    componentLinearVelScale=<1 1 1>
    physicsBodyDefinitions [{
      sphereCollisionRadius=0
      boundBone {boneName=dyn_shorts_r_02}
      boxExtents=<3 3 3>
      constraintSetup {angularTargetAxis=AxisZ, twistAxis=AxisY, angularLimitsMax=<0 0 30>, angularLimitsMin=<0 0 -30>, angularTarget=<0 0 1>}
      localJointOffset=<0 0 0>
     }]
    retargetingSettings {bUseAbsoluteAngle=true}
   }
   n10 = Anim Dynamics - Bone: dyn_shorts_r_03(ComponentPose=@n9, Alpha=1, ExternalForce="0,0,0") {
    alpha=0.9
    angularSpringConstant=1000
    bAngularSpring=true
    bGravityOverrideInSimSpace=true
    numSolverIterationsPostUpdate=4
    numSolverIterationsPreUpdate=16
    simulationSpace=Actor
    alphaBoolBlend {blendOption=Linear}
    boundBone {boneName=dyn_shorts_r_03}
    componentAppliedLinearAccClamp=<1000 1000 1000>
    componentLinearAccScale=<1 1 1>
    componentLinearVelScale=<1 1 1>
    physicsBodyDefinitions [{
      sphereCollisionRadius=0
      boundBone {boneName=dyn_shorts_r_03}
      boxExtents=<3 3 3>
      constraintSetup {angularTargetAxis=AxisZ, twistAxis=AxisY, angularLimitsMax=<0 0 30>, angularLimitsMin=<0 0 -30>, angularTarget=<0 0 1>}
      localJointOffset=<0 0 0>
     }]
    retargetingSettings {bUseAbsoluteAngle=true}
   }
   n11 = Anim Dynamics - Bone: dyn_shorts_r_04(ComponentPose=@n10, Alpha=1, ExternalForce="0,0,0") {
    alpha=0.9
    angularSpringConstant=1000
    bAngularSpring=true
    bGravityOverrideInSimSpace=true
    numSolverIterationsPostUpdate=4
    numSolverIterationsPreUpdate=16
    simulationSpace=Actor
    alphaBoolBlend {blendOption=Linear}
    boundBone {boneName=dyn_shorts_r_04}
    componentAppliedLinearAccClamp=<1000 1000 1000>
    componentLinearAccScale=<1 1 1>
    componentLinearVelScale=<1 1 1>
    physicsBodyDefinitions [{
      sphereCollisionRadius=0
      boundBone {boneName=dyn_shorts_r_04}
      boxExtents=<3 3 3>
      constraintSetup {angularTargetAxis=AxisZ, twistAxis=AxisY, angularLimitsMax=<0 0 30>, angularLimitsMin=<0 0 -30>, angularTarget=<0 0 1>}
      localJointOffset=<0 0 0>
     }]
    retargetingSettings {bUseAbsoluteAngle=true}
   }
   n12 = Anim Dynamics - Bone: dyn_shorts_r_06(ComponentPose=@n11, Alpha=1, ExternalForce="0,0,0") {
    alpha=0.9
    angularSpringConstant=1000
    bAngularSpring=true
    bGravityOverrideInSimSpace=true
    numSolverIterationsPostUpdate=4
    numSolverIterationsPreUpdate=16
    simulationSpace=Actor
    alphaBoolBlend {blendOption=Linear}
    boundBone {boneName=dyn_shorts_r_06}
    componentAppliedLinearAccClamp=<1000 1000 1000>
    componentLinearAccScale=<1 1 1>
    componentLinearVelScale=<1 1 1>
    physicsBodyDefinitions [{
      sphereCollisionRadius=0
      boundBone {boneName=dyn_shorts_r_06}
      boxExtents=<3 3 3>
      constraintSetup {angularTargetAxis=AxisZ, twistAxis=AxisY, angularLimitsMax=<0 0 30>, angularLimitsMin=<0 0 -30>, angularTarget=<0 0 1>}
      localJointOffset=<0 0 0>
     }]
    retargetingSettings {bUseAbsoluteAngle=true}
   }
   n13 = Anim Dynamics - Bone: dyn_sleeve_r_10(ComponentPose=@n12, Alpha=1, ExternalForce="0,0,0") {
    angularDampingOverride=1
    angularSpringConstant=5000
    bAngularSpring=true
    bGravityOverrideInSimSpace=true
    bOverrideAngularDamping=true
    gravityScale=0.35
    simulationSpace=RootRelative
    alphaBoolBlend {blendOption=Linear}
    alphaScaleBiasClamp {scale=0.2}
    boundBone {boneName=dyn_sleeve_r_10}
    componentLinearAccScale {z=0.6}
    componentLinearVelScale {z=0.6}
    physicsBodyDefinitions [{
      sphereCollisionRadius=0
      boundBone {boneName=dyn_sleeve_r_10}
      boxExtents=<10 20 20>
      constraintSetup {angularLimitsMax=<45 20 0>, angularLimitsMin=<-45 -20 0>, angularTarget=<1 0 0>}
      localJointOffset=<7 0 0>
     }]
   }
   n14 = Anim Dynamics - Bone: dyn_sleeve_r_09(ComponentPose=@n13, Alpha=1, ExternalForce="0,0,0") {
    angularDampingOverride=1
    angularSpringConstant=5000
    bAngularSpring=true
    bGravityOverrideInSimSpace=true
    bOverrideAngularDamping=true
    gravityScale=0.35
    simulationSpace=RootRelative
    alphaBoolBlend {blendOption=Linear}
    alphaScaleBiasClamp {scale=0.2}
    boundBone {boneName=dyn_sleeve_r_09}
    componentLinearAccScale {z=0.6}
    componentLinearVelScale {z=0.6}
    physicsBodyDefinitions [{
      sphereCollisionRadius=0
      boundBone {boneName=dyn_sleeve_r_09}
      boxExtents=<10 20 20>
      constraintSetup {angularLimitsMax=<45 20 0>, angularLimitsMin=<-45 -20 0>, angularTarget=<1 0 0>}
      localJointOffset=<7 0 0>
     }]
   }
   n15 = Anim Dynamics - Bone: dyn_sleeve_r_12(ComponentPose=@n14, Alpha=1, ExternalForce="0,0,0") {
    angularDampingOverride=1
    angularSpringConstant=5000
    bAngularSpring=true
    bGravityOverrideInSimSpace=true
    bOverrideAngularDamping=true
    gravityScale=0.35
    simulationSpace=RootRelative
    alphaBoolBlend {blendOption=Linear}
    alphaScaleBiasClamp {scale=0.2}
    boundBone {boneName=dyn_sleeve_r_12}
    componentLinearAccScale {z=0.6}
    componentLinearVelScale {z=0.6}
    physicsBodyDefinitions [{
      sphereCollisionRadius=0
      boundBone {boneName=dyn_sleeve_r_12}
      boxExtents=<10 20 20>
      constraintSetup {angularLimitsMax=<45 20 0>, angularLimitsMin=<-45 -20 0>, angularTarget=<1 0 0>}
      localJointOffset=<7 0 0>
     }]
   }
   n16 = Anim Dynamics - Bone: dyn_sleeve_r_11(ComponentPose=@n15, Alpha=1, ExternalForce="0,0,0") {
    angularDampingOverride=1
    angularSpringConstant=5000
    bAngularSpring=true
    bGravityOverrideInSimSpace=true
    bOverrideAngularDamping=true
    gravityScale=0.35
    simulationSpace=RootRelative
    alphaBoolBlend {blendOption=Linear}
    alphaScaleBiasClamp {scale=0.2}
    boundBone {boneName=dyn_sleeve_r_11}
    componentLinearAccScale {z=0.6}
    componentLinearVelScale {z=0.6}
    physicsBodyDefinitions [{
      sphereCollisionRadius=0
      boundBone {boneName=dyn_sleeve_r_11}
      boxExtents=<10 20 20>
      constraintSetup {angularLimitsMax=<45 20 0>, angularLimitsMin=<-45 -20 0>, angularTarget=<1 0 0>}
      localJointOffset=<7 0 0>
     }]
   }
   n17 = Anim Dynamics - Bone: dyn_sleeve_l_10(ComponentPose=@n16, Alpha=1, ExternalForce="0,0,0") {
    angularDampingOverride=1
    angularSpringConstant=5000
    bAngularSpring=true
    bGravityOverrideInSimSpace=true
    bOverrideAngularDamping=true
    gravityScale=0.35
    simulationSpace=RootRelative
    alphaBoolBlend {blendOption=Linear}
    alphaScaleBiasClamp {scale=0.2}
    boundBone {boneName=dyn_sleeve_l_10}
    componentLinearAccScale {z=0.6}
    componentLinearVelScale {z=0.6}
    physicsBodyDefinitions [{
      sphereCollisionRadius=0
      boundBone {boneName=dyn_sleeve_l_10}
      boxExtents=<10 20 20>
      constraintSetup {angularLimitsMax=<45 20 0>, angularLimitsMin=<-45 -20 0>, angularTarget=<1 0 0>}
      localJointOffset=<7 0 0>
     }]
   }
   n18 = Anim Dynamics - Bone: dyn_sleeve_l_09(ComponentPose=@n17, Alpha=1, ExternalForce="0,0,0") {
    angularDampingOverride=1
    angularSpringConstant=5000
    bAngularSpring=true
    bGravityOverrideInSimSpace=true
    bOverrideAngularDamping=true
    gravityScale=0.35
    simulationSpace=RootRelative
    alphaBoolBlend {blendOption=Linear}
    alphaScaleBiasClamp {scale=0.2}
    boundBone {boneName=dyn_sleeve_l_09}
    componentLinearAccScale {z=0.6}
    componentLinearVelScale {z=0.6}
    physicsBodyDefinitions [{
      sphereCollisionRadius=0
      boundBone {boneName=dyn_sleeve_l_09}
      boxExtents=<10 20 20>
      constraintSetup {angularLimitsMax=<45 20 0>, angularLimitsMin=<-45 -20 0>, angularTarget=<1 0 0>}
      localJointOffset=<7 0 0>
     }]
   }
   n19 = Anim Dynamics - Bone: dyn_sleeve_l_12(ComponentPose=@n18, Alpha=1, ExternalForce="0,0,0") {
    angularDampingOverride=1
    angularSpringConstant=5000
    bAngularSpring=true
    bGravityOverrideInSimSpace=true
    bOverrideAngularDamping=true
    gravityScale=0.35
    simulationSpace=RootRelative
    alphaBoolBlend {blendOption=Linear}
    alphaScaleBiasClamp {scale=0.2}
    boundBone {boneName=dyn_sleeve_l_12}
    componentLinearAccScale {z=0.6}
    componentLinearVelScale {z=0.6}
    physicsBodyDefinitions [{
      sphereCollisionRadius=0
      boundBone {boneName=dyn_sleeve_l_12}
      boxExtents=<10 20 20>
      constraintSetup {angularLimitsMax=<45 20 0>, angularLimitsMin=<-45 -20 0>, angularTarget=<1 0 0>}
      localJointOffset=<7 0 0>
     }]
   }
   n20 = Anim Dynamics - Bone: dyn_sleeve_l_11(ComponentPose=@n19, Alpha=1, ExternalForce="0,0,0") {
    angularDampingOverride=1
    angularSpringConstant=5000
    bAngularSpring=true
    bGravityOverrideInSimSpace=true
    bOverrideAngularDamping=true
    gravityScale=0.35
    simulationSpace=RootRelative
    alphaBoolBlend {blendOption=Linear}
    alphaScaleBiasClamp {scale=0.2}
    boundBone {boneName=dyn_sleeve_l_11}
    componentLinearAccScale {z=0.6}
    componentLinearVelScale {z=0.6}
    physicsBodyDefinitions [{
      sphereCollisionRadius=0
      boundBone {boneName=dyn_sleeve_l_11}
      boxExtents=<10 20 20>
      constraintSetup {angularLimitsMax=<45 20 0>, angularLimitsMin=<-45 -20 0>, angularTarget=<1 0 0>}
      localJointOffset=<7 0 0>
     }]
   }
   n21 = Anim Dynamics - Bone: lace_l_01(ComponentPose=@n20, Alpha=1, ExternalForce="0,0,0") {
    angularDampingOverride=0.9995
    angularSpringConstant=2500
    bAngularSpring=true
    bGravityOverrideInSimSpace=true
    bOverrideAngularDamping=true
    bUseGravityOverride=true
    alphaBoolBlend {blendOption=Linear}
    boundBone {boneName=lace_l_01}
    componentAppliedLinearAccClamp=<3000 3000 3000>
    componentLinearAccScale=<0.3 0.6 0.3>
    componentLinearVelScale=<0.3 0.3 0.3>
    gravityOverride {z=-980}
    physicsBodyDefinitions [{
      boundBone {boneName=lace_l_01}
      boxExtents=<10 10 10>
      constraintSetup {angularTargetAxis=AxisZ, angularLimitsMax=<0 50 0>, angularLimitsMin=<0 -40 0>, angularTarget=<0 0 1>}
      localJointOffset=<0 0 0>
     }]
   }
   n22 = Anim Dynamics - Bone: lace_l_02(ComponentPose=@n21, Alpha=1, ExternalForce="0,0,0") {
    angularDampingOverride=0.9995
    angularSpringConstant=2500
    bAngularSpring=true
    bGravityOverrideInSimSpace=true
    bOverrideAngularDamping=true
    bUseGravityOverride=true
    alphaBoolBlend {blendOption=Linear}
    boundBone {boneName=lace_l_02}
    componentAppliedLinearAccClamp=<3000 3000 3000>
    componentLinearAccScale=<0.3 0.6 0.3>
    componentLinearVelScale=<0.3 0.3 0.3>
    gravityOverride {z=-980}
    physicsBodyDefinitions [{
      boundBone {boneName=lace_l_02}
      boxExtents=<10 10 10>
      constraintSetup {angularTargetAxis=AxisZ, angularLimitsMax=<0 50 0>, angularLimitsMin=<0 -40 0>, angularTarget=<0 0 1>}
      localJointOffset=<0 0 0>
     }]
   }
   n23 = Anim Dynamics - Bone: lace_r_01(ComponentPose=@n22, Alpha=1, ExternalForce="0,0,0") {
    angularDampingOverride=0.9995
    angularSpringConstant=2500
    bAngularSpring=true
    bGravityOverrideInSimSpace=true
    bOverrideAngularDamping=true
    bUseGravityOverride=true
    alphaBoolBlend {blendOption=Linear}
    boundBone {boneName=lace_r_01}
    componentAppliedLinearAccClamp=<3000 3000 3000>
    componentLinearAccScale=<0.3 0.6 0.3>
    componentLinearVelScale=<0.3 0.3 0.3>
    gravityOverride {z=-980}
    physicsBodyDefinitions [{
      boundBone {boneName=lace_r_01}
      boxExtents=<10 10 10>
      constraintSetup {angularTargetAxis=AxisZ, angularLimitsMax=<0 50 0>, angularLimitsMin=<0 -40 0>, angularTarget=<0 0 1>}
      localJointOffset=<0 0 0>
     }]
   }
   n24 = Anim Dynamics - Bone: lace_r_02(ComponentPose=@n23, Alpha=1, ExternalForce="0,0,0") {
    angularDampingOverride=0.9995
    angularSpringConstant=2500
    bAngularSpring=true
    bGravityOverrideInSimSpace=true
    bOverrideAngularDamping=true
    bUseGravityOverride=true
    alphaBoolBlend {blendOption=Linear}
    boundBone {boneName=lace_r_02}
    componentAppliedLinearAccClamp=<3000 3000 3000>
    componentLinearAccScale=<0.3 0.6 0.3>
    componentLinearVelScale=<0.3 0.3 0.3>
    gravityOverride {z=-980}
    physicsBodyDefinitions [{
      boundBone {boneName=lace_r_02}
      boxExtents=<10 10 10>
      constraintSetup {angularTargetAxis=AxisZ, angularLimitsMax=<0 50 0>, angularLimitsMin=<0 -40 0>, angularTarget=<0 0 1>}
      localJointOffset=<0 0 0>
     }]
   }
   n25 = Control Rig!(Source=Component To Local(ComponentPose=@n24), Alpha) -> (Pose) {
    alphaBoolBlend {blendOption=Linear}
    controlRigAssetReference {blueprintRigClass=&/F_MED_BrightShimmer_Casual_Deform_CtrlRig.F_MED_BrightShimmer_Casual_Deform_CtrlRig_C}
   }
   Output Pose(Result=@n25) {layerGroup=None}
  ]}
 EventGraph {nodes [
   n1 = Event BlueprintUpdateAnimation()
   > Set Is lod2(Is lod2:bool="integer >= integer"(A=Get CurrentLOD(), B=2)) -> (Output_Get:bool)
   > n2 = Sequence()
   > n3 = Gravity Override State Function DEPRECATED(CharacterPartAnimInstance=Self-Reference(), GameTimeInSeconds=0, GravityOverrideParams=Get Gravity Override Params())
   > Set Gravity Override(Gravity Override=@n3)
   n4 = Get Mesh!() -> (Mesh)
   n5 = Get Overlay Mat Instance Dyanmic() -> (Overlay Mat Instance Dyanmic:Material Instance Dynamic Object Reference)
   n6 = Get Material Instance Dynamic() -> (Material Instance Dynamic:Material Instance Dynamic Object Reference)
   > Set WPO Globals Function(
    Material=@n6
    Overlay Material=@n5
    Skeletal Mesh=@n4
    "Local Noise Transform  Location"="Get Local Noise Transform  Location"() -> ("Local Noise Transform  Location":Vector)
    Local Noise Offset=Get Local Noise Offset()
   )
   > Set WPO Regional Function!(
    execute
    Material:Material Instance Dynamic Object Reference=@n6
    Overlay Material:Material Instance Dynamic Object Reference=@n5
    Skeletal Mesh:Skeletal Mesh Component Object Reference=@n4
    Velocity Transform:Name=calf_r
    Velocity Interp Speed:double=Get Velocity Interp Speed()
    Collision Transform:Name=calf_r
    Collision Transform Axis:EAxis Enum
    Invert Collision Transform Axis:bool=false
    Delta Time X:double=@n1.DeltaTimeX
    Previous Velocity Transform Location:"Vector (by ref)"=Get R2 Velocity Transform Location()
    Previous Velocity Transform Velocity:"Vector (by ref)"=Get R2 Velocity Transform Velocity()
    Region:int=1
   ) -> (then)
   > Set WPO Regional Function(
    Material=@n6
    Overlay Material=@n5
    Skeletal Mesh=@n4
    Velocity Transform=calf_l
    Velocity Interp Speed=Get Velocity Interp Speed()
    Collision Transform=calf_l
    Invert Collision Transform Axis=true
    Delta Time X=@n1.DeltaTimeX
    Previous Velocity Transform Location=Get L Previous Velocity Transform Location() -> (L Previous Velocity Transform Location:Vector)
    Previous Velocity Transform Velocity=Get L Previous Velocity Transform Velocity() -> (L Previous Velocity Transform Velocity:Vector)
    Region=2
   )
   > SetScalarParameterValue(self=[@n6, @n5], ParameterName=ClothSim_WindStrength, Value=Get WPO Wind() -> (WPO Wind:double))
   > SetScalarParameterValue(self=[@n6, @n5], ParameterName=ClothSim_TurbulenceNoiseTileDivisor, Value=Get Turbulence Noise() -> (Turbulence Noise:float))
   n7 = Branch(
    execute=@n2.then_1
    Condition=OR Boolean!(A=Get bIsSkydiveFloating(), B=Get bIsSkydiveDiveMode(), C:bool=Get bIsParachuteOpened()) -> (ReturnValue)
   )
   > Set Velocity Interp Speed(Velocity Interp Speed:double=40) -> (Output_Get:double)
   > Set WPO Wind(WPO Wind:double=1) -> (Output_Get:double)
   > Set Turbulence Noise(Turbulence Noise:float=80) -> (Output_Get:float)
   Set Velocity Interp Speed(execute=@n7.else, Velocity Interp Speed=0)
   > Set WPO Wind(WPO Wind=0)
   > Set Turbulence Noise(Turbulence Noise=0)
   Event OnCharacterCustomizationLoaded()
   n8 = GetPartSkeletalMeshComponent()
   > n9 = Set Mesh(Mesh:Skeletal Mesh Component Object Reference=@n8) -> (Output_Get:Skeletal Mesh Component Object Reference)
   > n10 = Get Dynamic Material Instance(SkeletalMeshComponent=@n9)
   > Set Material Instance Dynamic(Material Instance Dynamic:Material Instance Dynamic Object Reference=@n10) -> (Output_Get:Material Instance Dynamic Object Reference)
   > n11 = GetOrCreateDynamicOverlayMaterial(MeshComponent=@n8, MaterialName=None)
   > Set Overlay Mat Instance Dyanmic(Overlay Mat Instance Dyanmic:Material Instance Dynamic Object Reference=@n11) -> (Output_Get:Material Instance Dynamic Object Reference)
  ]}
]
variables [
 custom_cape_equipped {type=boolean}
 External Force {type=Vector}
 External Force_0 {type=Vector}
 Gravity Override {type=Vector}
 Gravity Override Params {type=GravityOverrideParamsStruct}
 Is lod2 {type=boolean}
 L Previous Velocity Transform Location {type=Vector}
 L Previous Velocity Transform Velocity {type=Vector}
 Local Noise Offset {type=Vector}
 "Local Noise Transform  Location" {type=Vector}
 Local Noise Transform Location {type=Vector}
 Mat Instance Dynamic {type=MaterialInstanceDynamic}
 Material Instance Dynamic {type=MaterialInstanceDynamic}
 Mesh {type=SkeletalMeshComponent}
 Overlay Mat Instance Dyanmic {type=MaterialInstanceDynamic}
 Overlay Mat Instance Dynamic {type=MaterialInstanceDynamic}
 pantlegAlphaFrontend {type=number}
 PlayerPawn {type=FortPlayerPawnAthena}
 R1 Velocity Transform Location {type=Vector}
 R1 Velocity Transform Velocity {type=Vector}
 R2 Velocity Transform Location {type=Vector}
 R2 Velocity Transform Velocity {type=Vector}
 R3 Velocity Transform Location {type=Vector}
 R3 Velocity Transform Velocity {type=Vector}
 R4 Velocity Transform Location {type=Vector}
 R4 Velocity Transform Velocity {type=Vector}
 Scarf Visible {type=boolean}
 SkelMesh {type=SkeletalMeshComponent}
 Turbulence Noise {type=number}
 Velocity Interp Speed {type=number}
 WPO Pho Wind Intensity {type=number}
 WPO Wind {type=number}
]
