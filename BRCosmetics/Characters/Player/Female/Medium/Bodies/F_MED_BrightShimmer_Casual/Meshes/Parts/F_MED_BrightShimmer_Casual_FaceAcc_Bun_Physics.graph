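# Physics asset for the hair bun face accessory of the F_MED_BrightShimmer_Casual female character, letting the bun move physically on the head.
# One capsule body on the dyn_ball_01 bone forms the bun, and three overlapping capsules on the head bone approximate the skull.
# A single constraint ties dyn_ball_01 to head with tight swing and tighter twist limits, restricting the bun to a small wobble.
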PhysicsAsset /BRCosmetics/Characters/Player/Female/Medium/Bodies/F_MED_BrightShimmer_Casual/Meshes/Parts/F_MED_BrightShimmer_Casual_FaceAcc_Bun_Physics
bodies {
 dyn_ball_01 [{shapeType=Capsule, length=6.5, radius=3.005, shapeName=dyn_ball_01_capsule, center=<3.4595 0 0>, rotation=<rot 0 -90 0>}]
 head [
  {shapeType=Capsule, length=17.5, radius=4.5, shapeName=head_capsule, center=<4.6458 -2.8149 0>, rotation=<rot 0 -0.0001 90>}
  {shapeType=Capsule, length=22, radius=5, shapeName=head_capsule, center=<9.7259 0.8713 0>}
  {shapeType=Capsule, length=11, radius=6.5, shapeName=head_capsule, center=<0 -1.3767 0>, rotation=<rot 90 180 -90>}
 ]
}
constraints [{bone1Name=dyn_ball_01, bone2Name=head, swing1LimitDegrees=8, swing2LimitDegrees=8, twistLimitDegrees=3}]
