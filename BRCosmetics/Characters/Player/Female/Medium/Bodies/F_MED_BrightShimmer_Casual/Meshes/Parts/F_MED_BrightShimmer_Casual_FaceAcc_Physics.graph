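# Physics asset for the BrightShimmer Casual medium female character's FaceAcc mesh, giving physics to its left and right curl bone chains, likely hair.
# Each curl is a chain of capsule bodies hung from the head, skipping left bone 06 and right bone 05.
# Constraints use narrow swing and near-zero twist limits, with swing widening toward the tips.
# A four-capsule head body and unconstrained upper-arm capsules provide collision shapes around the curls.
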PhysicsAsset /BRCosmetics/Characters/Player/Female/Medium/Bodies/F_MED_BrightShimmer_Casual/Meshes/Parts/F_MED_BrightShimmer_Casual_FaceAcc_Physics
bodies {
 dyn_curl_l_01 [{shapeType=Capsule, length=2, radius=2, shapeName=dyn_curl_l_01_capsule, rotation=<rot 90 0 0>}]
 dyn_curl_l_02 [{shapeType=Capsule, length=2, radius=2, shapeName=dyn_curl_l_02_capsule, rotation=<rot 90 0 0>}]
 dyn_curl_l_03 [{shapeType=Capsule, length=2, radius=2, shapeName=dyn_curl_l_03_capsule, rotation=<rot 90 0 0>}]
 dyn_curl_l_04 [{shapeType=Capsule, length=2, radius=2, shapeName=dyn_curl_l_04_capsule, rotation=<rot 90 0 0>}]
 dyn_curl_l_05 [{shapeType=Capsule, length=2, radius=2, shapeName=dyn_curl_l_05_capsule, rotation=<rot 70 0 -0.0002>}]
 dyn_curl_l_07 [{shapeType=Capsule, length=2.7093, radius=1.4585, shapeName=dyn_curl_l_07_capsule, center=<1.8129 0 -0.4189>, rotation=<rot 90 0 0>}]
 dyn_curl_r_01 [{shapeType=Capsule, length=2, radius=2, shapeName=dyn_curl_r_01_capsule, rotation=<rot 90 0 0>}]
 dyn_curl_r_02 [{shapeType=Capsule, length=2, radius=2, shapeName=dyn_curl_r_02_capsule, rotation=<rot 90 0 0>}]
 dyn_curl_r_03 [{shapeType=Capsule, length=2, radius=2, shapeName=dyn_curl_r_03_capsule, rotation=<rot 90 0 0>}]
 dyn_curl_r_04 [{shapeType=Capsule, length=2, radius=2, shapeName=dyn_curl_r_04_capsule, rotation=<rot 90 0 0>}]
 dyn_curl_r_06 [{shapeType=Capsule, length=3.9596, radius=1.7672, shapeName=dyn_curl_r_06_capsule, center=<3.4414 0 -0.4189>, rotation=<rot 90 0 0>}]
 head [
  {shapeType=Capsule, length=17.5, radius=4.5, shapeName=head_capsule, center=<4.6458 -2.8149 0>, rotation=<rot 0 -0.0001 90>}
  {shapeType=Capsule, length=22, radius=5, shapeName=head_capsule, center=<9.7259 0.8713 0>}
  {shapeType=Capsule, length=11, radius=6.5, shapeName=head_capsule, center=<0 -1.3767 0>, rotation=<rot 90 180 -90>}
  {shapeType=Capsule, length=21, radius=9, shapeName=head_capsule, center=<0 20.8784 0>}
 ]
 upperarm_l [{shapeType=Capsule, length=22.5, radius=4.5, shapeName=upperarm_l_capsule, center=<15 0 0>, rotation=<rot 90 0 0>}]
 upperarm_r [{shapeType=Capsule, length=22.5, radius=4.5, shapeName=upperarm_r_capsule, center=<-15 0 0>, rotation=<rot 90 0 0>}]
}
constraints [
 {bone1Name=dyn_curl_l_01, bone2Name=head, swing1LimitDegrees=20, swing2LimitDegrees=5, twistLimitDegrees=2}
 {bone1Name=dyn_curl_l_02, bone2Name=dyn_curl_l_01, swing1LimitDegrees=20, swing2LimitDegrees=10, twistLimitDegrees=2}
 {bone1Name=dyn_curl_l_03, bone2Name=dyn_curl_l_02, swing1LimitDegrees=20, swing2LimitDegrees=10, twistLimitDegrees=2}
 {bone1Name=dyn_curl_l_04, bone2Name=dyn_curl_l_03, swing1LimitDegrees=20, swing2LimitDegrees=10, twistLimitDegrees=2}
 {bone1Name=dyn_curl_l_05, bone2Name=dyn_curl_l_04, swing1LimitDegrees=20, swing2LimitDegrees=20, twistLimitDegrees=2}
 {bone1Name=dyn_curl_r_01, bone2Name=head, swing1LimitDegrees=20, swing2LimitDegrees=10, twistLimitDegrees=2}
 {bone1Name=dyn_curl_r_02, bone2Name=dyn_curl_r_01, swing1LimitDegrees=20, swing2LimitDegrees=10, twistLimitDegrees=2}
 {bone1Name=dyn_curl_r_03, bone2Name=dyn_curl_r_02, swing1LimitDegrees=20, swing2LimitDegrees=10, twistLimitDegrees=2}
 {bone1Name=dyn_curl_r_04, bone2Name=dyn_curl_r_03, swing1LimitDegrees=20, swing2LimitDegrees=20, twistLimitDegrees=2}
 {bone1Name=dyn_curl_l_07, bone2Name=dyn_curl_l_05, swing1LimitDegrees=20, swing2LimitDegrees=20, twistLimitDegrees=0}
 {bone1Name=dyn_curl_r_06, bone2Name=dyn_curl_r_04, swing1LimitDegrees=20, swing2LimitDegrees=20, twistLimitDegrees=0}
]
