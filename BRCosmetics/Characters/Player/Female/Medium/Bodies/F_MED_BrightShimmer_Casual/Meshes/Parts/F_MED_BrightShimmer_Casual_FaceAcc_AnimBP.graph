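# Post-process style secondary-motion AnimBP for the BrightShimmer Casual female character's face accessory mesh, including a hair bun.
# It copies the attached parent's pose, layering in clavicles and upper arms from the owner's skeletal mesh, cached at initialization.
# Two chained bone-space RigidBody passes follow: partial-strength bun physics, then accessory physics.
# An AnimDynamics ball node and another bun RigidBody node sit unconnected.
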AnimBlueprint /BRCosmetics/Characters/Player/Female/Medium/Bodies/F_MED_BrightShimmer_Casual/Meshes/Parts/F_MED_BrightShimmer_Casual_FaceAcc_AnimBP
target_skeleton=&/F_MED_BrightShimmer_Casual_FaceAcc_Skeleton
graphs [
 AnimGraph {nodes [
   Anim Dynamics - Bone: dyn_ball_01(Alpha=0.15, ExternalForce="0,0,0") {
    angularBiasOverride=1
    angularDampingOverride=1
    angularSpringConstant=1400
    bAngularSpring=true
    bGravityOverrideInSimSpace=true
    bOverrideAngularDamping=true
    bUseGravityOverride=true
    simulationSpace=RootRelative
    alphaBoolBlend {blendOption=Linear}
    boundBone {boneName=dyn_ball_01}
    componentAppliedLinearAccClamp=<1000 1000 1000>
    componentLinearAccScale=<1 1 1>
    componentLinearVelScale=<1 1 1>
    gravityOverride {z=-980}
    physicsBodyDefinitions [{
      boundBone {boneName=dyn_ball_01}
      boxExtents=<3 3 3>
      constraintSetup {angularLimitsMax=<0 20 25>, angularLimitsMin=<0 -20 -25>, angularTarget=<1 0 0>}
      localJointOffset=<0 0 0>
     }]
    relativeSpaceBone {boneName=head}
   }
   RigidBody(Alpha=0.8) {
    bOverrideWorldGravity=true
    overridePhysicsAsset=&/F_MED_BrightShimmer_Casual_FaceAcc_Bun_Physics
    simulationSpace=BaseBoneSpace
    alphaBoolBlend {blendOption=Linear}
    baseBoneRef {boneName=spine_05}
    componentLinearAccScale=<0.3 0.3 0.3>
    componentLinearVelScale=<0.3 0.3 0.3>
    overrideWorldGravity {z=-980}
   }
   n1 = Layered blend per bone(
    BasePose=Copy Pose From Mesh() {bUseAttachedParent=true}
    BlendPoses_0=Copy Pose From Mesh(SourceMeshComponent=Get Body mesh() -> (Body mesh:Skeletal Mesh Component Object Reference))
   ) {layerSetup [{branchFilters [{boneName=clavicle_l}, {boneName=clavicle_r}, {boneName=upperarm_l}, {boneName=upperarm_r}]}]}
   n2 = RigidBody(ComponentPose=Local To Component(LocalPose=@n1), Alpha=0.5) {
    bClampLinearTranslationLimitToRefPose=true
    bOverrideWorldGravity=true
    overridePhysicsAsset=&F_MED_BrightShimmer_Casual_FaceAcc_Bun_Physics
    simulationSpace=BaseBoneSpace
    alphaBoolBlend {blendOption=Linear}
    baseBoneRef {boneName=spine_05}
    componentAppliedLinearAccClamp=<6000 6000 6000>
    componentLinearAccScale=<0.25 0.25 0.25>
    componentLinearVelScale=<0.25 0.25 0.25>
    overrideWorldGravity {z=-980}
   }
   n3 = RigidBody(ComponentPose=@n2) {
    bClampLinearTranslationLimitToRefPose=true
    bOverrideWorldGravity=true
    overridePhysicsAsset=&/F_MED_BrightShimmer_Casual_FaceAcc_Physics
    simulationSpace=BaseBoneSpace
    alphaBoolBlend {blendOption=Linear}
    baseBoneRef {boneName=spine_02}
    componentAppliedLinearAccClamp=<3000 3000 3000>
    componentLinearAccScale=<0.3 0.2 0.2>
    componentLinearVelScale=<0.3 0.2 0.2>
    overrideWorldGravity {z=-980}
    simSpaceSettings {maxAngularAcceleration=1000, maxAngularVelocity=3, maxLinearAcceleration=250, maxLinearVelocity=500, velocityScaleZ=0, worldAlpha=1}
   }
   Output Pose(Result=Component To Local(ComponentPose=@n3)) {layerGroup=None}
  ]}
 EventGraph {nodes [
   Event BlueprintInitializeAnimation()
   > Set Body mesh(Body mesh:Skeletal Mesh Component Object Reference=GetOwnerSkeletalMeshComponent()) -> (Output_Get:Skeletal Mesh Component Object Reference)
   Event BlueprintUpdateAnimation()
  ]}
]
variables [
 Body mesh {type=SkeletalMeshComponent}
 External Force {type=Vector}
 gravity_override {type=Vector}
 Override World Gravity {type=Vector, default=<0 0 -980>}
]
